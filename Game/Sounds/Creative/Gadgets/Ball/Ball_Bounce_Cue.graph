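# Sound cue for the bounce of the Creative ball gadget, played through a ball bounce attenuation asset in the gadgets items sound class.
# It layers two random picks in a mixer: one of five dedicated ball bounce waves at full volume and one of five balloon pickaxe impact waves at a much lower level.
# A modulator at the root varies the combined result on each play.
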SoundCue /Game/Sounds/Creative/Gadgets/Ball/Ball_Bounce_Cue
root=Modulator_0
nodes [
 {id=Mixer_1, children [Random_1,Random_0], properties {inputVolume [1,0.3]}}
 {id=Modulator_0, children [Mixer_1]}
 {id=Random_0, children [WavePlayer_1,WavePlayer_2,WavePlayer_3,WavePlayer_6,WavePlayer_9], properties {weights [1,1,1,1,1]}}
 {id=Random_1, children [WavePlayer_0,WavePlayer_4,WavePlayer_5,WavePlayer_7,WavePlayer_8], properties {weights [1,1,1,1,1]}}
 {id=WavePlayer_0, properties {soundWaveAssetPtr=&/Ball_Bounce_01}}
 {id=WavePlayer_1, properties {soundWaveAssetPtr=/Game/Athena/Sounds/Weapons/PickAxes/Balloon/PickAxe_Balloon_EnemyImpact_06}}
 {id=WavePlayer_2, properties {soundWaveAssetPtr=/Game/Athena/Sounds/Weapons/PickAxes/Balloon/PickAxe_Balloon_EnemyImpact_04}}
 {id=WavePlayer_3, properties {soundWaveAssetPtr=/Game/Athena/Sounds/Weapons/PickAxes/Balloon/PickAxe_Balloon_EnemyImpact_03}}
 {id=WavePlayer_4, properties {soundWaveAssetPtr=&/Ball_Bounce_02}}
 {id=WavePlayer_5, properties {soundWaveAssetPtr=&/Ball_Bounce_03}}
 {id=WavePlayer_6, properties {soundWaveAssetPtr=/Game/Athena/Sounds/Weapons/PickAxes/Balloon/PickAxe_Balloon_EnemyImpact_02}}
 {id=WavePlayer_7, properties {soundWaveAssetPtr=&/Ball_Bounce_05}}
 {id=WavePlayer_8, properties {soundWaveAssetPtr=&/Ball_Bounce_06}}
 {id=WavePlayer_9, properties {soundWaveAssetPtr=/Game/Athena/Sounds/Weapons/PickAxes/Balloon/PickAxe_Balloon_EnemyImpact_01}}
]
properties {
 attenuationSettings=/Game/Sounds/Attenuation/Fort_Ball_Bounce_Athena_Att
 maxDistance=7000
 soundClassObject=&Fort_Gadgets_Items
 volumeMultiplier=1.1
}
