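# Sound cue for the pop of a Creative ball gadget.
# A Random node picks one of four equally weighted pop waves and feeds a Modulator that slightly varies pitch on each play.
# It uses the gadget-activate attenuation and gadget items sound class, with boosted volume and a concurrency override that blocks new instances once the limit is reached.
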SoundCue /Game/Sounds/Creative/Gadgets/Ball/Ball_Pop_Cue
root=Modulator_0
nodes [
 {id=Modulator_0, children [Random_1], properties {pitchMax=1.1, pitchMin=0.9}}
 {id=Random_1, children [WavePlayer_3,WavePlayer_5,WavePlayer_6,WavePlayer_7], properties {weights [1,1,1,1]}}
 {id=WavePlayer_3, properties {soundWaveAssetPtr=&/Ball_Pop_01}}
 {id=WavePlayer_5, properties {soundWaveAssetPtr=&/Ball_Pop_02}}
 {id=WavePlayer_6, properties {soundWaveAssetPtr=&/Ball_Pop_03}}
 {id=WavePlayer_7, properties {soundWaveAssetPtr=&/Ball_Pop_04}}
]
properties {
 attenuationSettings=&Fort_GadgetActivate_Att
 bOverrideConcurrency=true
 maxDistance=4000
 soundClassObject=&Fort_Gadgets_Items
 volumeMultiplier=1.5
 concurrencyOverrides {resolutionRule=PreventNew}
}
