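# Looping sound cue for the rolling noise of the Creative ball gadget.
# A looping roll wave feeds a Continuous Modulator that drives both pitch and volume from a Speed parameter, with pitch held to a narrow range around normal.
# A default-setting Modulator at the root adds variation on top.
# The cue plays in the gadget items sound class at reduced volume and shares the ball bounce attenuation.
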SoundCue /Game/Sounds/Creative/Gadgets/Ball/Ball_Roll_Loop_Cue
root=Modulator_0
nodes [
 {
  id=ModulatorContinuous_0
  children [WavePlayer_1]
  properties {
   pitchModulationParams {maxInput=1.1, maxOutput=1.1, minInput=0.8, minOutput=0.8, parameterName=Speed}
   volumeModulationParams {parameterName=Speed}
  }
 }
 {id=Modulator_0, children [ModulatorContinuous_0]}
 {id=WavePlayer_1, properties {bLooping=true, soundWaveAssetPtr=&/Ball_Roll_Loop_01}}
]
properties {
 attenuationSettings=/Game/Sounds/Attenuation/Fort_Ball_Bounce_Athena_Att
 maxDistance=7000
 soundClassObject=&Fort_Gadgets_Items
 volumeMultiplier=0.35
}
