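# Sound cue for the spawn sound of the Creative Ball gadget, routed to the gadget items sound class.
# A mixer plays three layers at once: two Ball_Spawn waves through modulators and a borrowed balloon pickaxe swing wave played directly.
# The first spawn layer is pitched slightly up by a fixed amount and mixed a little lower, while the second uses default modulation.
# Spatialization uses the shared ball bounce attenuation asset.
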SoundCue /Game/Sounds/Creative/Gadgets/Ball/Ball_Spawn_Cue
root=Mixer_0
nodes [
 {id=Mixer_0, children [Modulator_0, Modulator_1, WavePlayer_12], properties {inputVolume [0.8,1,1]}}
 {id=Modulator_0, children [WavePlayer_2], properties {pitchMax=1.07, pitchMin=1.07}}
 {id=Modulator_1, children [WavePlayer_3]}
 {id=WavePlayer_12, properties {soundWaveAssetPtr=/Game/Athena/Sounds/Weapons/PickAxes/Balloon/PA_Balloon_Swing_03}}
 {id=WavePlayer_2, properties {soundWaveAssetPtr=&/Ball_Spawn_02}}
 {id=WavePlayer_3, properties {soundWaveAssetPtr=&/Ball_Spawn_03}}
]
properties {
 attenuationSettings=/Game/Sounds/Attenuation/Fort_Ball_Bounce_Athena_Att
 maxDistance=7000
 soundClassObject=&Fort_Gadgets_Items
 volumeMultiplier=0.6
}
